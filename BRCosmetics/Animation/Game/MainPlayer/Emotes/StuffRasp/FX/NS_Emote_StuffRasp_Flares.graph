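# Niagara effect for the StuffRasp emote that renders two burning flares attached to bones or sockets of a skeletal mesh.
# All emitters use particle partitioning and a shared system mesh reader to split between the two attach points.
# Flames combines sprite, mesh and light renderers, with system and emitter age curves plus waveforms driving ignition, flicker and burnout.
# Sprite smoke and ribbon trails use time-of-day tint, while two spark emitters spray bursts along socket rotation.
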NiagaraSystem /BRCosmetics/Animation/Game/MainPlayer/Emotes/StuffRasp/FX/NS_Emote_StuffRasp_Flares
dependencies {
 usedDataInterfaces [
  &NiagaraDataInterfaceArrayInt32
  &NiagaraDataInterfaceCurve
  &NiagaraDataInterfaceColorCurve
  &NiagaraDataInterfaceSpriteRendererInfo
  &NiagaraDataInterfaceSkeletalMesh
 ]
 usedDynamicInputs [&VectorLength]
 usedModules [
  &SystemState
  &self:SystemSpawnScript.SystemScriptSource.SystemScriptGraph.NiagaraNodeAssignment_0.SetVariables_#1
  &self:SystemSpawnScript.SystemScriptSource.SystemScriptGraph.NiagaraNodeAssignment_1.SetVariables_#2
  &self:SystemSpawnScript.SystemScriptSource.SystemScriptGraph.NiagaraNodeAssignment_2.SetVariables_#3
  &self:SystemSpawnScript.SystemScriptSource.SystemScriptGraph.NiagaraNodeAssignment_3.SetVariables_#4
  &self:SystemSpawnScript.SystemScriptSource.SystemScriptGraph.NiagaraNodeAssignment_4.SetVariables_#5
 ]
}
emitters {
 Flames {
  emitterUpdateScript [
   &EmitterState(
    Inactive Response=::NewEnumerator1
    Life Cycle Mode=::NewEnumerator1
    Loop Behavior=::NewEnumerator1
    Loop Duration=3.9833
    UseLoopDelay=false
   )
   &SpawnBurst_Instantaneous(Spawn Count=2, Use Spawn Probability=false, UseLoopCountLimit=false)
  ]
  particleSpawnScript [
   &PartitionParticles()
   &InitializeParticle(
    Color=<rgba 3.7708 0.0779 0 1>
    Lifetime=&RandomRangeFloat(Minimum=1.5, Maximum=2)
    Mesh Scale Mode=::NewEnumerator3
    Mesh Uniform Scale=0.3
    Sprite Size Mode=::NewEnumerator3
    Uniform Sprite Size=70
    UsePositionOffset=false
    Write Material Random=true
   )
   /SetVariables() {isSetParameters=true}
  ]
  particleUpdateScript [
   &ParticleState(
    Kill Particles When Lifetime Has Elapsed=false
    Let Infinitely Lived Particles Die When Emitter Deactivates=true
    Loop Particles Lifetime=true
   )
   &SkeletalMeshLocation(
    Bone / Socket Index=&SelectIntFromArray(Direct Array Index=$Particles.PartitionParticles.Partition, Int Selection Array=DataInterface() {propertyValues {IntData [0,1]}})
    Bone Sampling Mode=::NewEnumerator4
    MaskSpawnGroup=false
    Sampled Position Offset=<0 10 0>
    Skeletal Mesh=$System.Skeletal Mesh Reader
    Use Sampled Mesh Attribute as Ribbon ID=false
   )
   &ScaleSpriteSize(
    Uniform Curve Scale=$System.FlareFlameDuration
    Uniform Curve Sprite Scale=DataInterface() {propertyValues {Curve {keys [
        <at 0 1 RCIM_Cubic>
        {arriveTangent=-0.6622, interpMode=RCIM_Cubic, leaveTangent=-0.6622, tangentMode=RCTM_User, time=0.1, value=0.775}
        {arriveTangent=-0.0208, interpMode=RCIM_Cubic, leaveTangent=-0.0208, tangentMode=RCTM_User, time=0.2, value=0.85}
        {arriveTangent=0.3451, interpMode=RCIM_Cubic, leaveTangent=0.3451, tangentMode=RCTM_User, time=0.3, value=0.875}
        {arriveTangent=-0.0022, interpMode=RCIM_Cubic, leaveTangent=-0.0022, tangentMode=RCTM_User, time=0.4, value=0.95}
        {arriveTangent=-0.0096, interpMode=RCIM_Cubic, leaveTangent=-0.0096, tangentMode=RCTM_User, time=0.5, value=0.875}
        {arriveTangent=-0.037, interpMode=RCIM_Cubic, leaveTangent=-0.037, tangentMode=RCTM_User, time=0.6, value=0.9}
        {arriveTangent=-0.0446, interpMode=RCIM_Cubic, leaveTangent=-0.0446, tangentMode=RCTM_User, time=0.7, value=0.75}
        {arriveTangent=-0.0333, interpMode=RCIM_Cubic, leaveTangent=-0.0333, tangentMode=RCTM_User, time=0.8, value=0.95}
        {arriveTangent=-0.0145, interpMode=RCIM_Cubic, leaveTangent=-0.0145, tangentMode=RCTM_User, time=0.9, value=0.875}
        <at 1 1 RCIM_Cubic>
       ]}}}
   )
   &ScaleSpriteSize(
    Uniform Curve Index=$Emitter.Age
    Uniform Curve Sprite Scale=DataInterface() {propertyValues {Curve {keys [
        {arriveTangent=5.0894, interpMode=RCIM_Cubic, leaveTangent=5.0894, tangentMode=RCTM_User, time=0, value=0.4}
        {arriveTangent=-0.2694, interpMode=RCIM_Cubic, leaveTangent=-0.2694, tangentMode=RCTM_User, time=0.15, value=1.5716}
        <at 0.3 1 RCIM_Cubic>
       ]}}}
   ) {module=ScaleSpriteSize001}
   &ScaleMeshSize(Scale Factor=&MakeVector(
     X=$System.FlareFlameDuration
     Y=$System.FlareFlameDuration
     Z=&FloatFromCurve(
      FloatCurve=DataInterface() {propertyValues {Curve {keys [
          {arriveTangent=5.0894, interpMode=RCIM_Cubic, leaveTangent=5.0894, tangentMode=RCTM_User, time=0, value=0.4}
          {arriveTangent=0.2071, interpMode=RCIM_Cubic, leaveTangent=0.2071, tangentMode=RCTM_User, time=0.2, value=1.8}
          {arriveTangent=-0.0762, interpMode=RCIM_Cubic, leaveTangent=-0.0762, time=0.5, value=1}
          {arriveTangent=-0.0727, interpMode=RCIM_Cubic, leaveTangent=-0.0727, time=10.7, value=1}
          <at 11.5 0.2 RCIM_Cubic>
         ]}}}
      CurveIndex=$Emitter.Age
     )
    ))
   /SetVariables(Particles.ColorFlare:LinearColor=<rgba 40 4.3393 4.5493 1>) {isSetParameters=true}
   /SetVariables(Particles.SpriteFlareSize:Vector2f=&Multiply_Vector2D_ByFloat(
     Float=&FloatFromCurve(
      FloatCurve=DataInterface() {propertyValues {Curve {keys [<at 0.23 0.3924>,<at 0.3119 0.6626>,<at 0.4338 0.6715>,<at 0.6 1>,<at 2.5446 1>,<at 3.085 0>]}}}
      CurveIndex=$System.Age
     )
     Vector2D=&Multiply_Vector2D_ByFloat(
      Float=$System.FlareDuration
      Vector2D=&MakeVector2D(
       X=&Multiply_Float(
        A=60
        B=&FloatFromCurve(
         FloatCurve=DataInterface() {propertyValues {Curve {keys [
             <at 0 1 RCIM_Cubic>
             {arriveTangent=0.0166, interpMode=RCIM_Cubic, leaveTangent=0.0166, tangentMode=RCTM_User, time=0.1, value=0.725}
             {arriveTangent=-0.0112, interpMode=RCIM_Cubic, leaveTangent=-0.0112, tangentMode=RCTM_User, time=0.2, value=0.875}
             {arriveTangent=0.0059, interpMode=RCIM_Cubic, leaveTangent=0.0059, tangentMode=RCTM_User, time=0.3, value=0.85}
             {arriveTangent=-0.0022, interpMode=RCIM_Cubic, leaveTangent=-0.0022, tangentMode=RCTM_User, time=0.4, value=0.95}
             {arriveTangent=-0.0096, interpMode=RCIM_Cubic, leaveTangent=-0.0096, tangentMode=RCTM_User, time=0.5, value=0.8125}
             {arriveTangent=-0.037, interpMode=RCIM_Cubic, leaveTangent=-0.037, tangentMode=RCTM_User, time=0.6, value=0.85}
             {arriveTangent=0.0048, interpMode=RCIM_Cubic, leaveTangent=0.0048, tangentMode=RCTM_User, time=0.7, value=0.725}
             {arriveTangent=-0.011, interpMode=RCIM_Cubic, leaveTangent=-0.011, tangentMode=RCTM_User, time=0.8, value=0.95}
             {arriveTangent=-0.0145, interpMode=RCIM_Cubic, leaveTangent=-0.0145, tangentMode=RCTM_User, time=0.9, value=0.875}
             <at 1 1 RCIM_Cubic>
            ]}}}
         Scale Curve=&FloatFromCurve(
          FloatCurve=DataInterface() {propertyValues {Curve {keys [
              {arriveTangent=5.0894, interpMode=RCIM_Cubic, leaveTangent=5.0894, tangentMode=RCTM_User, time=0, value=0.4}
              {arriveTangent=1.2, interpMode=RCIM_Cubic, leaveTangent=1.2, time=0.0857, value=11.9185}
              <at 0.5 1 RCIM_Cubic>
             ]}}}
          CurveIndex=$Emitter.Age
         )
        )
       )
       Y=15
      )
     )
    )) {isSetParameters=true}
   /SetVariables(Particles.ColorFlame:LinearColor=&MakeLinearColorFromVectorAndFloat(
     "Vector (RGB)"=&MakeVectorFromLinearColorRGB(Color=<rgba 97.6 0.7878 0 1>)
     "Float (Alpha)"=&Multiply_Float(
      A=&FloatFromCurve(FloatCurve=DataInterface() {propertyValues {Curve {keys [
           <at 0 1 RCIM_Cubic>
           {arriveTangent=0.0166, interpMode=RCIM_Cubic, leaveTangent=0.0166, tangentMode=RCTM_User, time=0.1, value=0.6}
           {arriveTangent=-0.0112, interpMode=RCIM_Cubic, leaveTangent=-0.0112, tangentMode=RCTM_User, time=0.2, value=0.875}
           {arriveTangent=0.0059, interpMode=RCIM_Cubic, leaveTangent=0.0059, tangentMode=RCTM_User, time=0.3, value=0.7}
           {arriveTangent=-0.0022, interpMode=RCIM_Cubic, leaveTangent=-0.0022, tangentMode=RCTM_User, time=0.4, value=0.95}
           {arriveTangent=-0.0096, interpMode=RCIM_Cubic, leaveTangent=-0.0096, tangentMode=RCTM_User, time=0.5, value=0.75}
           {arriveTangent=-0.037, interpMode=RCIM_Cubic, leaveTangent=-0.037, tangentMode=RCTM_User, time=0.6, value=0.775}
           {arriveTangent=0.0048, interpMode=RCIM_Cubic, leaveTangent=0.0048, tangentMode=RCTM_User, time=0.7, value=0.65}
           {arriveTangent=-0.011, interpMode=RCIM_Cubic, leaveTangent=-0.011, tangentMode=RCTM_User, time=0.8, value=0.9}
           {arriveTangent=-0.0145, interpMode=RCIM_Cubic, leaveTangent=-0.0145, tangentMode=RCTM_User, time=0.9, value=0.825}
           <at 1 1 RCIM_Cubic>
          ]}}})
      B=&FloatFromCurve(
       FloatCurve=DataInterface() {propertyValues {Curve {keys [
           {arriveTangent=5.0894, interpMode=RCIM_Cubic, leaveTangent=5.0894, tangentMode=RCTM_User, time=0, value=0.4}
           {arriveTangent=1.2, interpMode=RCIM_Cubic, leaveTangent=1.2, time=0.2, value=10}
           <at 0.5 1 RCIM_Cubic>
          ]}}}
       CurveIndex=$Emitter.Age
       Scale Curve=$System.FlareFlameDuration
      )
     )
    )) {isSetParameters=true}
   &CameraOffset(Camera Offset Amount=6)
   /SetVariables(
    Particles.ColorLight:LinearColor=&MakeLinearColorFromVectorAndFloat(
     "Vector (RGB)"=&MakeVectorFromLinearColorRGB(Color=<rgba 5 0.3626 0 1>)
     "Float (Alpha)"=&Multiply_Float(
      A=$System.FlareFlameDuration
      B=&FloatFromCurve(
       FloatCurve=DataInterface() {propertyValues {Curve {keys [<at 0 0>,<at 0.0257 1>,<at 0.0375 0>,<at 0.317 0>,<at 0.3487 1.0012>,<at 0.3634 0>,<at 0.4513 0>,<at 0.458 0.8384>,<at 0.5192 0.1236>,<at 0.6 1>,<at 1 1>,<at 3.3 1>,<at 3.32 0>,<at 3.4154 0>,<at 3.4201 0.9691>,<at 3.528 0.9665>,<at 3.539 0>]}}}
       CurveIndex=$System.Age
      )
     )
    )
    Particles.GlowColor=&Multiply_LinearColorByFloat(
     Linear Color=<rgba 1.895 0.1374 0 1>
     Float=&FN_FrontEndFloatModifier(
      Gameplay=&Waveform("Waveform [1]"=::NewEnumerator0, "[1] Frequency"=4.264, "[1] Amplitude Min/Max"=<0.3 0.5>)
      Front End=&Waveform("Waveform [1]"=::NewEnumerator0, "[1] Frequency"=4.264, "[1] Amplitude Min/Max"=<0.2 0.3>)
     )
    )
    Particles.GlowSize=&Multiply_Vector2D_ByFloat(
     Float=&FloatFromCurve(FloatCurve=DataInterface() {propertyValues {Curve {keys [<at 0.7591 1>,<at 1 0>]}}}, Scale Curve=50)
     Vector2D=<1 1>
    )
   ) {isSetParameters=true}
   &ScaleColor(
    Scale RGB=&VectorFromFloat(Value=&FloatFromCurve(
      FloatCurve=DataInterface() {propertyValues {Curve {keys [
          {arriveTangent=5.0894, interpMode=RCIM_Cubic, leaveTangent=5.0894, tangentMode=RCTM_User, time=0, value=0.4}
          {arriveTangent=2, interpMode=RCIM_Cubic, leaveTangent=2, time=0.1, value=6}
          <at 0.3 1 RCIM_Cubic>
         ]}}}
      CurveIndex=$Emitter.Age
      Scale Curve=$System.FlareFlameDuration
     ))
    ScaleA=true
    ScaleRGB=true
   )
   &DynamicMaterialParameters(
    Index 1 Param 4=&FloatFromCurve(
     FloatCurve=DataInterface() {propertyValues {Curve {keys [
         {arriveTangent=-0.0917, interpMode=RCIM_Cubic, leaveTangent=-0.0917, tangentMode=RCTM_User, time=0, value=1}
         {arriveTangent=-0.031, interpMode=RCIM_Cubic, leaveTangent=-0.031, tangentMode=RCTM_User, time=9, value=0.5}
         <at 11 0.1 RCIM_Cubic>
        ]}}}
     CurveIndex=$Emitter.Age
     Scale Curve=1.5
    )
    Module.Index1Float0Write=true
    Module.Index1Float1Write=true
    Module.Index1Float2Write=true
    Module.Index1Float3Write=true
    Write Parameter Index 0=::NewEnumerator0
    Write Parameter Index 1=::NewEnumerator1
   )
   &ScaleColor(
    Scale Alpha=&Multiply_Float(
     A=&FloatFromCurve(
      FloatCurve=DataInterface() {propertyValues {Curve {keys [<at 0 0>,<at 0.0121 0.5354>,<at 0.0264 0.0362>,<at 0.044 0.1842>,<at 0.0461 -0.024>,<at 0.0858 0.0935>,<at 0.1218 -0.0144>,<at 0.317 0>,<at 0.3271 1.0012>,<at 0.3715 0>,<at 0.4275 -0.0002>,<at 0.4278 0.9045>,<at 0.4581 -0.0064>,<at 0.4871 0.7378>,<at 0.5548 0.5371>,<at 0.6 1>,<at 1 1>,<at 3.3 1>,<at 3.32 0>,<at 3.4154 0>,<at 3.4201 3>,<at 3.528 2>,<at 3.539 0>]}}}
      CurveIndex=$System.Age
     )
     B=&Waveform(
      "Waveform [1]"=::NewEnumerator0
      "[1] Frequency"=&RandomRangeFloat(Minimum=2, Maximum=2.5)
      "[1] Amplitude Min/Max"=<0.35 1>
     )
    )
    ScaleA=true
    ScaleRGB=false
   ) {module=ScaleColor001}
   /SetVariables(Particles.Flame Update:LinearColor=&MultiplyLinearColor(
     Linear Color=$Particles.ColorFlame
     Scale Factor=&MakeVector4_FromLinearColor(LinearColor=&MakeLinearColorFromVectorAndFloat(
       "Vector (RGB)"=<1 1 1>
       "Float (Alpha)"=&FloatFromCurve(
        FloatCurve=DataInterface() {propertyValues {Curve {keys [<at 0.317 0>,<at 0.3438 0.6001>,<at 0.3715 0>,<at 0.4144 0.0924>,<at 0.4597 0.8602>,<at 0.4977 0.1951>,<at 0.6 1>,<at 1 1>,<at 3.0034 0.941>,<at 3.1816 0.0354>,<at 3.2819 0.0118>,<at 3.3163 0.7804>,<at 3.3709 -0.0118>]}}}
        CurveIndex=$System.Age
        Scale Curve=&Waveform(
         "Waveform [1]"=::NewEnumerator0
         "[1] Frequency"=&RandomRangeFloat(Minimum=2, Maximum=2.5)
         "[1] Amplitude Min/Max"=<0.35 1>
        )
       )
      ))
    )) {isSetParameters=true}
   /SetVariables(Particles.Flare Update:LinearColor=&MultiplyLinearColor(
     Linear Color=$Particles.ColorFlare
     Scale Factor=&MakeVector4_FromLinearColor(LinearColor=&MakeLinearColorFromVectorAndFloat(
       "Vector (RGB)"=<1 1 1>
       "Float (Alpha)"=&Multiply_Float(
        A=&FloatFromCurve(
         FloatCurve=DataInterface() {propertyValues {Curve {keys [<at 0.317 0>,<at 0.3271 0.803>,<at 0.3715 0>,<at 0.4015 0.0064>,<at 0.4122 0.8948>,<at 0.4477 0.431>,<at 0.6 1>,<at 1 1>,<at 3.3 1>,<at 3.32 0>,<at 3.4154 0>,<at 3.4201 2>,<at 3.528 0.9665>,<at 3.539 0>]}}}
         CurveIndex=$System.Age
        )
        B=1
       )
      ))
    )) {isSetParameters=true}
   &ScaleSpriteSize(
    Scale Sprite Size Mode=::NewEnumerator0
    Uniform Scale Factor=&Multiply_Float(
     A=&FloatFromCurve(
      FloatCurve=DataInterface() {propertyValues {Curve {keys [<at 0.0835 2.52>,<at 0.4338 0.6715>,<at 0.6 1>,<at 3.271 1>,<at 3.3564 0.6288>,<at 3.478 0.7514>]}}}
      CurveIndex=$System.Age
     )
     B=&Waveform(
      "Waveform [1]"=::NewEnumerator0
      "[1] Frequency"=&RandomRangeFloat(Minimum=2, Maximum=2.5)
      "[1] Amplitude Min/Max"=<0.75 0.496>
     )
    )
   ) {module=ScaleSpriteSize002}
   /SetVariables(Particles.FlareSize:Vector2f=&Multiply_Vector2D_ByFloat(
     Float=&FloatFromCurve(
      FloatCurve=DataInterface() {propertyValues {Curve {
         defaultValue=1
         keys [
          {arriveTangent=-5.1605, interpMode=RCIM_Cubic, leaveTangent=-5.1605, tangentMode=RCTM_User, time=0.0053, value=1.5302}
          {arriveTangent=-0.3728, interpMode=RCIM_Cubic, leaveTangent=-0.3728, tangentMode=RCTM_User, time=0.0892, value=0.4704}
          {arriveTangent=-0.0405, interpMode=RCIM_Cubic, leaveTangent=-0.0405, tangentMode=RCTM_User, time=1, value=0.1654}
         ]
        }}}
      Scale Curve=71.3794
     )
     Vector2D=<1 1>
    )) {isSetParameters=true}
  ]
  rendererClasses [
   &NiagaraSpriteRendererProperties
   &NiagaraLightRendererProperties
   &NiagaraMeshRendererProperties
  ]
  renderers [
   {rendererClass=&NiagaraSpriteRendererProperties}
   {rendererClass=&NiagaraSpriteRendererProperties, rendererIndex=1}
   {rendererClass=&NiagaraLightRendererProperties, rendererIndex=2}
   {rendererClass=&NiagaraMeshRendererProperties, rendererIndex=3}
   {rendererClass=&NiagaraSpriteRendererProperties, rendererIndex=4}
   {rendererClass=&NiagaraSpriteRendererProperties, rendererIndex=5}
  ]
 }
 Smoke {
  emitterUpdateScript [
   &EmitterState(Life Cycle Mode=::NewEnumerator1, Loop Behavior=::NewEnumerator1, Loop Duration=3.5, UseLoopDelay=false)
   &SpawnRate(SpawnGroup=1, SpawnRate=75, Use Spawn Probability=false)
   &SpawnRate(SpawnRate=75, Use Spawn Probability=false) {module=SpawnRate001}
  ]
  particleSpawnScript [
   &PartitionParticles(Write Partition to Ribbon ID=::NewEnumerator0)
   /SetVariables(Particles.VisibilityTag=&SpawnGroup()) {isSetParameters=true}
   &InitializeParticle(
    Color=&FNTimeOfDayColor(Night Color=<rgba 0.3847 0.4712 1 0.15>, Morning Color=<rgba 1 1 1 0.15>, Day Color=<rgba 1 1 1 0.15>, Evening Color=<rgba 1 1 1 0.15>)
    Lifetime=&MergeHelper_FloatSwitch(
     0=&RandomRangeFloat(Minimum=0.9, Maximum=1.3)
     1=&RandomRangeFloat(Minimum=0.7, Maximum=0.9)
     2=?
     3=?
     Index=&SpawnGroup()
     Num=::NewEnumerator0
    )
    Mass Mode=::NewEnumerator3
    Sprite Rotation Mode=::NewEnumerator2
    Sprite Size Mode=::NewEnumerator1
    Uniform Sprite Size Max=30
    Uniform Sprite Size Min=18
    UsePositionOffset=false
    Write Material Random=true
   )
   &SkeletalMeshLocation(
    Bone / Socket Index=&SelectIntFromArray(Direct Array Index=$Particles.PartitionParticles.Partition, Int Selection Array=DataInterface() {propertyValues {IntData [0,1]}})
    Bone Sampling Mode=::NewEnumerator4
    MaskSpawnGroup=false
    Sampled Position Offset=<0 10 0>
    Skeletal Mesh=$System.Skeletal Mesh Reader
   )
   &ShapeLocation(MaskSpawnGroup=false, Override Seed=false, Sphere Radius=8, UseOwnerScale=true)
   &AddVelocity(
    Cone Angle=32
    Cone Axis=<0 0 1>
    Distribution Along Cone Axis=0.25
    Module.UseConeFalloff=true
    Module.UseVelDistribution=true
    Override Seed=false
    Speed Falloff From Cone Axis=0.5
    Velocity Mode=::NewEnumerator2
    Velocity Speed=&RandomRangeFloat(Minimum=42.7535, Maximum=90.2995)
   )
  ]
  particleUpdateScript [
   &ParticleState()
   &Drag(Drag=0.25, Use Linear Drag=true, Use Rotational Drag=false)
   &ScaleSpriteSize(Scale Factor=&Vector2DFromFloat(Value=&Multiply_Float(
      A=&FloatFromCurve(
       FloatCurve=DataInterface() {propertyValues {Curve {keys [
           {arriveTangent=2.5043, interpMode=RCIM_Cubic, leaveTangent=2.5043, tangentMode=RCTM_User, time=0.0016, value=0.2}
           {arriveTangent=0.2677, interpMode=RCIM_Cubic, leaveTangent=0.2677, tangentMode=RCTM_User, time=0.8464, value=1.0206}
           {arriveTangent=-236.4009, interpMode=RCIM_Cubic, leaveTangent=-236.4009, tangentMode=RCTM_User, time=0.85, value=0}
          ]}}}
       Scale Curve=2.25
      )
      B=&FloatFromCurve(
       FloatCurve=DataInterface() {propertyValues {Curve {
          defaultValue=1
          keys [
           {arriveTangent=4.0943, interpMode=RCIM_Cubic, leaveTangent=4.0943, tangentMode=RCTM_User, time=0.2462, value=1}
           {tangentMode=RCTM_User, time=1, value=2}
          ]
         }}}
       CurveIndex=$System.NormalizedLoopAge
      )
     )))
   &ScaleColor(
    Scale Alpha=&Multiply_Float(
     A=&FloatFromCurve(FloatCurve=DataInterface() {propertyValues {Curve {keys [<at 0.05 1 RCIM_Cubic>,<at 0.85 0>]}}})
     B=&FloatFromCurve(
      FloatCurve=DataInterface() {propertyValues {Curve {
         defaultValue=1
         keys [{tangentMode=RCTM_User, time=0.2707, value=1}, <at 0.3648 0.4319>, <at 0.6284 0.2211>, {tangentMode=RCTM_User, time=1, value=0}]
        }}}
      CurveIndex=$System.NormalizedLoopAge
     )
    )
    Scale RGB=&MergeHelper_VectorSwitch(
     0=&MakeVectorFromLinearColorRGB(Color=&MultiplyLinearColor(
       Linear Color=&ColorFromCurve(ColorCurve=DataInterface() {propertyValues {
          AlphaCurve {keys [<at 0.0274 1>]}
          BlueCurve {keys [<at 0.0094 0>,<at 0.1083 0>,<at 0.256 0.5781>,<at 0.937 0.2448>]}
          GreenCurve {keys [<at 0.0094 0.3955>,<at 0.1083 0.048>,<at 0.256 0.5048>,<at 0.937 0.2041>]}
          RedCurve {keys [<at 0.0094 20>,<at 0.1083 2>,<at 0.256 0.4405>,<at 0.937 0.2221>]}
         }})
       Scale Factor=&MakeVector4_FromLinearColor(LinearColor=&ColorFromCurve(
         ColorCurve=DataInterface() {propertyValues {
           AlphaCurve {keys []}
           BlueCurve {keys [<at 0.252 1 RCIM_Cubic>, {arriveTangent=-2.1229, interpMode=RCIM_Cubic, leaveTangent=-2.1229, time=0.5211, value=0.1156}, <at 0.711 0.0257>]}
           GreenCurve {keys [<at 0.252 1 RCIM_Cubic>, {arriveTangent=-2.1277, interpMode=RCIM_Cubic, leaveTangent=-2.1277, time=0.5211, value=0.1114}, <at 0.711 0.0234>]}
           RedCurve {keys [<at 0.252 1 RCIM_Cubic>, {arriveTangent=-2.1175, interpMode=RCIM_Cubic, leaveTangent=-2.1175, time=0.5211, value=0.1656}, <at 0.711 0.0281>]}
          }}
         CurveIndex=$System.NormalizedLoopAge
        ))
      ))
     1=&Multiply_VectorByFloat(
      Vector=&MakeVectorFromLinearColorRGB(Color=&MultiplyLinearColor(
        Linear Color=&ColorFromCurve(ColorCurve=DataInterface() {propertyValues {
           AlphaCurve {keys [<at 0.0307 1>]}
           BlueCurve {keys [<at 0.0094 0>,<at 0.0854 0.0495>,<at 0.2018 0.0708>,<at 0.4023 0.4248>]}
           GreenCurve {keys [<at 0.0094 0.0628>,<at 0.0854 0.0767>,<at 0.2018 0.0677>,<at 0.4023 0.4265>]}
           RedCurve {keys [<at 0.0094 0.8021>,<at 0.0854 0.4375>,<at 0.2018 0.2031>,<at 0.4023 0.4531>]}
          }})
        Scale Factor=&MakeVector4_FromLinearColor(LinearColor=&ColorFromCurve(
          ColorCurve=DataInterface() {propertyValues {
            AlphaCurve {keys []}
            BlueCurve {keys [<at 0.252 1 RCIM_Cubic>,<at 0.4678 0.0531>,<at 0.711 0.022>]}
            GreenCurve {keys [<at 0.252 1 RCIM_Cubic>,<at 0.4678 0.0446>,<at 0.711 0.0218>]}
            RedCurve {keys [<at 0.252 1 RCIM_Cubic>,<at 0.4678 0.0625>,<at 0.711 0.0281>]}
           }}
          CurveIndex=$System.NormalizedLoopAge
         ))
       ))
      Float=1
     )
     2=<0 0 0>
     3=<0 0 0>
     4=<0 0 0>
     Index=&SpawnGroup()
     Num=::NewEnumerator3
    )
    ScaleA=true
    ScaleRGB=true
   )
   &SolveForcesAndVelocity(Clamp Velocity=false, Limit Acceleration=false)
   &SubUVAnimation(
    Output Alternate Renderer Binding=true
    "Sprite Renderer "=DataInterface() {propertyValues {SpriteRenderer=&self:Smoke_0.NiagaraSpriteRendererProperties_1}}
    Start Frame Offset:NiagaraInt32=&RandomRangeInt(Minimum=0, Maximum=5)
    SubUV Lookup Index Scale=2.1054
    UseEndFrame=false
    UseStartFrame=false
   )
   &DynamicMaterialParameters(
    Index 0 Param 1=&FloatFromCurve(FloatCurve=DataInterface() {propertyValues {Curve {keys [<at 0 0 RCIM_Cubic>, {arriveTangent=0.4364, tangentMode=RCTM_User, time=1.0011, value=0.2}]}}})
    Index 0 Param 2=&FloatFromCurve(FloatCurve=DataInterface() {propertyValues {Curve {keys [
         {arriveTangent=0.0453, interpMode=RCIM_Cubic, tangentMode=RCTM_Break, time=0, value=-1}
         {arriveTangent=0.0604, interpMode=RCIM_Cubic, leaveTangent=0.0604, tangentMode=RCTM_User, time=1, value=-1}
        ]}}})
    Index 0 Param 3=&FloatFromCurve(FloatCurve=DataInterface() {propertyValues {Curve {keys [{interpMode=RCIM_Cubic, leaveTangent=4.6779, time=0.2851, value=0}]}}})
    Index 1 Param 1=&FloatFromCurve(
     FloatCurve=DataInterface() {propertyValues {Curve {keys [<at 0 0>,<at 0.25 1>]}}}
     Scale Curve=&RandomRangeFloat(Minimum=2, Maximum=1)
    )
    Index 1 Param 2=&FloatFromCurve(
     FloatCurve=DataInterface() {propertyValues {Curve {keys [
         {interpMode=RCIM_Cubic, leaveTangent=3.7422, tangentMode=RCTM_Break, time=0, value=0.2}
         {interpMode=RCIM_Cubic, leaveTangent=0.9066, time=1, value=2}
        ]}}}
     Scale Curve=&RandomRangeFloat(Minimum=0.5, Maximum=1)
    )
    Index 1 Param 3=&FloatFromCurve(FloatCurve=DataInterface() {propertyValues {Curve {keys [
         {interpMode=RCIM_Cubic, leaveTangent=2.8445, tangentMode=RCTM_Break, time=0, value=0.2}
         {interpMode=RCIM_Cubic, leaveTangent=63.2978, time=0.25, value=1}
        ]}}})
    Index 1 Param 4=&FloatFromCurve(
     FloatCurve=DataInterface() {propertyValues {Curve {keys [<at 0 1>,<at 1 0.8>]}}}
     Scale Curve=&RandomRangeFloat(Minimum=0.5, Maximum=1)
    )
    Index 2 Param 1=&FloatFromCurve(FloatCurve=DataInterface() {propertyValues {Curve {keys [{interpMode=RCIM_Cubic, leaveTangent=3.2309, tangentMode=RCTM_Break, time=0.1, value=0}, <at 1 1>]}}})
    Index 2 Param 2=$Particles.MaterialRandom
    Index 2 Param 3=&FloatFromCurve(FloatCurve=DataInterface() {propertyValues {Curve {keys [
         {interpMode=RCIM_Cubic, tangentMode=RCTM_Break, time=0, value=1}
         {interpMode=RCIM_Cubic, leaveTangent=11.1233, tangentMode=RCTM_Break, time=1, value=2}
        ]}}})
    Index 2 Param 4=0.75
    Module.Param0WriteEnabled=true
    Module.Param0WriteEnabled1=true
    Module.Param0WriteEnabled2=true
    Module.Param1WriteEnabled=false
    Module.Param1WriteEnabled1=true
    Module.Param1WriteEnabled2=true
    Module.Param2WriteEnabled=true
    Module.Param2WriteEnabled1=true
    Module.Param2WriteEnabled2=true
    Module.Param3WriteEnabled=true
    Module.Param3WriteEnabled1=true
    Module.Param3WriteEnabled2=true
    Write Parameter Index 1=true
    Write Parameter Index 2=true
   )
  ]
  rendererClasses [&NiagaraSpriteRendererProperties]
  renderers [
   {rendererClass=&NiagaraSpriteRendererProperties}
   {rendererClass=&NiagaraSpriteRendererProperties, rendererIndex=1}
  ]
 }
 Smoke_Ribbon {
  emitterUpdateScript [
   &EmitterState(Life Cycle Mode=::NewEnumerator1, Loop Behavior=::NewEnumerator1, Loop Duration=3.5, UseLoopDelay=false)
   &SpawnRate(
    SpawnRate=&Multiply_Float(
     A=100
     B=&FloatFromCurve(
      FloatCurve=DataInterface() {propertyValues {Curve {keys [
          {arriveTangentWeight=0.1817, interpMode=RCIM_Cubic, leaveTangentWeight=0.6939, tangentMode=RCTM_User, time=1.1743, value=1}
          {arriveTangent=-0.3007, interpMode=RCIM_Cubic, leaveTangent=-0.3007, time=3.6026, value=0.8456}
          <at 4.5 0 RCIM_Cubic>
         ]}}}
      CurveIndex=$Engine.System.Age
     )
    )
    Use Spawn Probability=false
   )
  ]
  particleSpawnScript [
   &PartitionParticles()
   &InitializeParticle(
    Color=&FNTimeOfDayColor(Night Color=<rgba 0.5659 0.5929 1 1>, Morning Color=<rgba 1 1 1 1>, Day Color=<rgba 1 1 1 1>, Evening Color=<rgba 1 1 1 1>)
    Lifetime=1.3
    Ribbon Width=40
    Ribbon Width Mode=::NewEnumerator1
    UsePositionOffset=false
    Write Material Random=true
   )
   &AddVelocity(Velocity=<0 0 101.9972>)
   &ApplyInitialForces()
   &SkeletalMeshLocation(
    Bone / Socket Index=&SelectIntFromArray(Direct Array Index=$Particles.PartitionParticles.Partition, Int Selection Array=DataInterface() {propertyValues {IntData [0,1]}})
    Bone Sampling Mode=::NewEnumerator4
    MaskSpawnGroup=false
    Sampled Position Offset=<0 10 0>
    Skeletal Mesh=$System.Skeletal Mesh Reader
   )
  ]
  particleUpdateScript [
   &ParticleState()
   &Drag(Drag=0.5, Use Linear Drag=true, Use Rotational Drag=true) {enabled=false}
   &SolveForcesAndVelocity(Clamp Velocity=false, Limit Acceleration=false)
   &DynamicMaterialParameters(
    Index 0 Param 1=-1
    Index 0 Param 4=1
    Module.Index0Float0Write=true
    Module.Index0Float1Write=true
    Module.Index0Float2Write=true
    Module.Index0Float3Write=true
   )
   &ScaleRibbonWidth(Ribbon Width Scale=&FloatFromCurve(FloatCurve=DataInterface() {propertyValues {Curve {
        defaultValue=1
        keys [
         <at 0.0042 0.4456 RCIM_Cubic>
         {arriveTangent=8.1687, interpMode=RCIM_Cubic, leaveTangent=8.1687, tangentMode=RCTM_User, time=0.1003, value=0.8631}
         {arriveTangent=5.0857, interpMode=RCIM_Cubic, leaveTangent=5.0857, tangentMode=RCTM_User, time=0.2394, value=1.611}
         <at 1.0002 2.5758 RCIM_Cubic>
        ]
       }}}))
   &ScaleColor(
    Scale Alpha=&Multiply_Float(
     A=&FloatFromCurve(FloatCurve=DataInterface() {propertyValues {Curve {
         defaultValue=1
         keys [
          {interpMode=RCIM_Cubic, tangentMode=RCTM_User, time=0.0221, value=0.36}
          {arriveTangent=-0.0257, interpMode=RCIM_Cubic, leaveTangent=-0.0257, time=0.0649, value=0.9936}
          {arriveTangent=-1.9935, interpMode=RCIM_Cubic, leaveTangent=-1.9935, tangentMode=RCTM_User, time=0.1574, value=0.3565}
          {interpMode=RCIM_Cubic, tangentMode=RCTM_User, time=0.7024, value=0.0129}
         ]
        }}})
     B=&FloatFromCurve(
      FloatCurve=DataInterface() {propertyValues {Curve {keys [
          {arriveTangent=0.7641, interpMode=RCIM_Cubic, leaveTangent=0.7641, tangentMode=RCTM_User, time=0.67, value=0.8}
          {arriveTangentWeight=0.1681, interpMode=RCIM_Cubic, leaveTangentWeight=0.7352, tangentMode=RCTM_User, time=1.1743, value=1}
          {arriveTangentWeight=0.7925, interpMode=RCIM_Cubic, leaveTangentWeight=0.1009, tangentMode=RCTM_User, time=3.3799, value=1}
          <at 3.6609 0 RCIM_Cubic>
         ]}}}
      CurveIndex=$Engine.System.Age
     )
    )
    Scale RGB=&MakeVectorFromLinearColorRGB(Color=&ColorFromCurve(ColorCurve=DataInterface() {propertyValues {
        AlphaCurve {keys []}
        BlueCurve {keys [<at 0.0094 0.0017>,<at 0.0387 0.1739>,<at 0.0874 0.5781>,<at 0.4126 0.5448>]}
        GreenCurve {keys [<at 0.0094 0>,<at 0.0387 0.1489>,<at 0.0874 0.5048>,<at 0.4126 0.4752>]}
        RedCurve {keys [<at 0.0094 5>,<at 0.0387 1.3438>,<at 0.0874 0.4405>,<at 0.4126 0.4171>]}
       }}))
    ScaleA=true
    ScaleRGB=true
   )
  ]
  rendererClasses [&NiagaraRibbonRendererProperties]
  renderers [{rendererClass=&NiagaraRibbonRendererProperties}]
 }
 Sparks {
  emitterUpdateScript [
   &EmitterState(
    Delay First Loop Only=true
    Life Cycle Mode=::NewEnumerator1
    Loop Behavior=::NewEnumerator2
    Loop Count=12
    Loop Delay=3.3
    Loop Duration=0.116
    UseLoopDelay=true
   )
   &SpawnRate(SpawnRate=60, Use Spawn Probability=false)
   &SpawnBurst_Instantaneous(
    Spawn Count=&RandomRangeInt(Minimum=3, Maximum=5)
    Use Spawn Probability=false
    UseLoopCountLimit=false
   )
  ]
  particleSpawnScript [
   &PartitionParticles()
   &InitializeParticle(
    Color=<rgba 20 1.0082 0 1>
    Lifetime Max=0.2
    Lifetime Min=0.1
    Lifetime Mode=::NewEnumerator1
    Sprite Size Mode=::NewEnumerator1
    Uniform Sprite Size Max=1.3908
    Uniform Sprite Size Min=1
    UsePositionOffset=false
    Write Material Random=true
   )
   &SkeletalMeshLocation(
    Bone / Socket Index=&SelectIntFromArray(Direct Array Index=$Particles.PartitionParticles.Partition, Int Selection Array=DataInterface() {propertyValues {IntData [0,1]}})
    Bone Sampling Mode=::NewEnumerator4
    MaskSpawnGroup=false
    Orientation Sampling=::NewEnumerator1
    Sampled Position Offset=<0 10 0>
    Skeletal Mesh=$System.Skeletal Mesh Reader
   )
   &AddVelocity(
    Additional Quaternion Rotation=false
    Cone Angle=68.1021
    Cone Axis=<0 90 0>
    Module.UseConeFalloff=false
    Module.UseVelDistribution=false
    Override Seed=false
    Rotation Mode=::NewEnumerator3
    Rotation Quaternion:Quat4f=$Particles.SkeletalMeshLocation.SampledRotation
    Velocity Mode=::NewEnumerator2
    Velocity Speed=72.4339
   )
   &ShapeLocation(MaskSpawnGroup=false, Override Seed=false, Sphere Radius=4.3312, UseOwnerScale=true)
  ]
  particleUpdateScript [
   &ParticleState()
   &Drag(
    Drag=&FloatFromCurve(
     FloatCurve=DataInterface() {propertyValues {Curve {
        defaultValue=1
        keys [
         {arriveTangent=4.1785, interpMode=RCIM_Cubic, leaveTangent=4.1785, tangentMode=RCTM_User, time=0, value=0}
         {arriveTangent=1.1179, interpMode=RCIM_Cubic, leaveTangent=1.1179, tangentMode=RCTM_User, time=0.2042, value=0.6228}
         {arriveTangent=0.0224, interpMode=RCIM_Cubic, leaveTangent=0.0224, tangentMode=RCTM_User, time=1, value=1}
        ]
       }}}
     Scale Curve=5
    )
    Use Linear Drag=true
    Use Rotational Drag=true
   )
   &GravityForce(Gravity=<0 0 -12.068>)
   &InheritSourceMovement(ApplyForce=false, ApplyPosition=true, ApplyPositionFalloff=false)
   &CurlNoiseForce(Bias Noise=false, Noise Strength=60, Pan Noise=false)
   &SolveForcesAndVelocity(Clamp Velocity=false, Limit Acceleration=false)
   &ScaleSpriteSizeBySpeed(Max Scale Factor=<1 5.0476>, Min Scale Factor=<1 3>)
   &DynamicMaterialParameters(
    Index 0 Param 1=-1
    Index 0 Param 4=1
    Module.Index0Float0Write=true
    Module.Index0Float1Write=true
    Module.Index0Float2Write=true
    Module.Index0Float3Write=true
   )
   &ScaleRibbonWidth(Ribbon Width Scale=&FloatFromCurve(FloatCurve=DataInterface() {propertyValues {Curve {
        defaultValue=1
        keys [
         <at 0.0042 0.4456 RCIM_Cubic>
         {arriveTangent=8.1687, interpMode=RCIM_Cubic, leaveTangent=8.1687, tangentMode=RCTM_User, time=0.1003, value=0.8631}
         {arriveTangent=5.0857, interpMode=RCIM_Cubic, leaveTangent=5.0857, tangentMode=RCTM_User, time=0.2394, value=1.611}
         <at 1.0002 2.5758 RCIM_Cubic>
        ]
       }}}))
   &ScaleColor(
    Scale Alpha=&Multiply_Float(
     A=&FloatFromCurve(FloatCurve=DataInterface() {propertyValues {Curve {
         defaultValue=1
         keys [
          {interpMode=RCIM_Cubic, tangentMode=RCTM_User, time=0.0221, value=0.36}
          {arriveTangent=-0.0257, interpMode=RCIM_Cubic, leaveTangent=-0.0257, time=0.0649, value=0.9936}
          {arriveTangent=-1.9935, interpMode=RCIM_Cubic, leaveTangent=-1.9935, tangentMode=RCTM_User, time=0.1574, value=0.3565}
          {interpMode=RCIM_Cubic, tangentMode=RCTM_User, time=0.7024, value=0.0129}
         ]
        }}})
     B=&FloatFromCurve(
      FloatCurve=DataInterface() {propertyValues {Curve {keys [
          {arriveTangent=0.7641, interpMode=RCIM_Cubic, leaveTangent=0.7641, tangentMode=RCTM_User, time=0.67, value=0.8}
          {arriveTangentWeight=0.1681, interpMode=RCIM_Cubic, leaveTangentWeight=0.7352, tangentMode=RCTM_User, time=1.1743, value=1}
          {arriveTangentWeight=0.7925, interpMode=RCIM_Cubic, leaveTangentWeight=0.1009, tangentMode=RCTM_User, time=3.3799, value=1}
         ]}}}
      CurveIndex=$Engine.System.Age
     )
    )
    Scale RGB=&MakeVectorFromLinearColorRGB(Color=&ColorFromCurve(ColorCurve=DataInterface() {propertyValues {
        AlphaCurve {keys []}
        BlueCurve {keys [<at 0.0028 1>,<at 0.0267 1 RCIM_Cubic>,<at 0.1061 1>,<at 0.1924 0.901>,<at 0.2694 0.574>,<at 0.3542 0.5781>,<at 0.6976 0.5448>,<at 0.9101 0.2448>]}
        GreenCurve {keys [
          <at 0.0028 0.6428>
          {arriveTangent=0.8673, interpMode=RCIM_Cubic, leaveTangent=0.8673, time=0.0267, value=0.8408}
          <at 0.1061 0.7324>
          <at 0.1924 0.8118>
          <at 0.2694 0.5537>
          <at 0.3542 0.5048>
          <at 0.6976 0.4752>
          <at 0.9101 0.2041>
         ]}
        RedCurve {keys [
          <at 0.0028 0.1861>
          {arriveTangent=2.9892, interpMode=RCIM_Cubic, leaveTangent=2.9892, time=0.0267, value=0.4676}
          <at 0.1061 0.4948>
          <at 0.1924 0.6664>
          <at 0.2694 0.4987>
          <at 0.3542 0.4405>
          <at 0.6976 0.4171>
          <at 0.9101 0.2221>
         ]}
       }}))
    ScaleA=true
    ScaleRGB=true
   )
  ]
  rendererClasses [&NiagaraSpriteRendererProperties]
  renderers [{rendererClass=&NiagaraSpriteRendererProperties}]
 }
 Sparks2 {
  emitterUpdateScript [
   &EmitterState(Life Cycle Mode=::NewEnumerator1, Loop Behavior=::NewEnumerator1, Loop Duration=0.9, UseLoopDelay=false)
   &SpawnBurst_Instantaneous(Spawn Count=10, Use Spawn Probability=false, UseLoopCountLimit=false)
  ]
  particleSpawnScript [
   &PartitionParticles()
   &InitializeParticle(
    Color=<rgba 20 1.0082 0 1>
    Lifetime Max=0.3
    Lifetime Min=0.15
    Lifetime Mode=::NewEnumerator1
    Sprite Size Mode=::NewEnumerator1
    Uniform Sprite Size Max=1.3908
    Uniform Sprite Size Min=1
    UsePositionOffset=false
    Write Material Random=true
   )
   &SkeletalMeshLocation(
    Bone / Socket Index=&SelectIntFromArray(Direct Array Index=$Particles.PartitionParticles.Partition, Int Selection Array=DataInterface() {propertyValues {IntData [0,1]}})
    Bone Sampling Mode=::NewEnumerator4
    MaskSpawnGroup=false
    Orientation Sampling=::NewEnumerator1
    Sampled Position Offset=<0 10 0>
    Skeletal Mesh=$System.Skeletal Mesh Reader
   )
   &AddVelocity(
    Additional Quaternion Rotation=false
    Cone Angle=68.1021
    Cone Axis=&MergeHelper_VectorSwitch(0=<0 0 90>, 1=<0 0 -90>, 2=?, 3=?, Index=$Particles.PartitionParticles.Partition, Num=::NewEnumerator0)
    Module.UseConeFalloff=false
    Module.UseVelDistribution=false
    Override Seed=false
    Rotation Mode=::NewEnumerator3
    Rotation Quaternion:Quat4f=$Particles.SkeletalMeshLocation.SampledRotation
    Velocity Mode=::NewEnumerator2
    Velocity Speed=320.1182
   )
   &ShapeLocation(MaskSpawnGroup=false, Override Seed=false, Sphere Radius=5.0049, UseOwnerScale=true)
  ]
  particleUpdateScript [
   &ParticleState()
   &Drag(Drag=0.5, Use Linear Drag=true, Use Rotational Drag=true) {enabled=false}
   &GravityForce(Gravity=<0 0 -172.1512>)
   &InheritSourceMovement(ApplyForce=false, ApplyPosition=true, ApplyPositionFalloff=false)
   &SolveForcesAndVelocity(Clamp Velocity=false, Limit Acceleration=false)
   &ScaleSpriteSizeBySpeed(Max Scale Factor=<1 9>, Min Scale Factor=<1 6.4211>)
   &DynamicMaterialParameters(
    Index 0 Param 1=-1
    Index 0 Param 4=1
    Module.Index0Float0Write=true
    Module.Index0Float1Write=true
    Module.Index0Float2Write=true
    Module.Index0Float3Write=true
   )
   &ScaleRibbonWidth(Ribbon Width Scale=&FloatFromCurve(FloatCurve=DataInterface() {propertyValues {Curve {
        defaultValue=1
        keys [
         <at 0.0042 0.4456 RCIM_Cubic>
         {arriveTangent=8.1687, interpMode=RCIM_Cubic, leaveTangent=8.1687, tangentMode=RCTM_User, time=0.1003, value=0.8631}
         {arriveTangent=5.0857, interpMode=RCIM_Cubic, leaveTangent=5.0857, tangentMode=RCTM_User, time=0.2394, value=1.611}
         <at 1.0002 2.5758 RCIM_Cubic>
        ]
       }}}))
   &ScaleColor(
    Scale Alpha=&Multiply_Float(
     A=&FloatFromCurve(FloatCurve=DataInterface() {propertyValues {Curve {
         defaultValue=1
         keys [
          {interpMode=RCIM_Cubic, tangentMode=RCTM_User, time=0.0221, value=0.36}
          {arriveTangent=-0.0257, interpMode=RCIM_Cubic, leaveTangent=-0.0257, time=0.0649, value=0.9936}
          {arriveTangent=-1.9935, interpMode=RCIM_Cubic, leaveTangent=-1.9935, tangentMode=RCTM_User, time=0.1574, value=0.3565}
          {interpMode=RCIM_Cubic, tangentMode=RCTM_User, time=0.7024, value=0.0129}
         ]
        }}})
     B=&FloatFromCurve(
      FloatCurve=DataInterface() {propertyValues {Curve {keys [
          {arriveTangent=0.7641, interpMode=RCIM_Cubic, leaveTangent=0.7641, tangentMode=RCTM_User, time=0.67, value=0.8}
          {arriveTangentWeight=0.1681, interpMode=RCIM_Cubic, leaveTangentWeight=0.7352, tangentMode=RCTM_User, time=1.1743, value=1}
          {arriveTangentWeight=0.7925, interpMode=RCIM_Cubic, leaveTangentWeight=0.1009, tangentMode=RCTM_User, time=3.3799, value=1}
         ]}}}
      CurveIndex=$Engine.System.Age
     )
    )
    Scale RGB=&MakeVectorFromLinearColorRGB(Color=&ColorFromCurve(ColorCurve=DataInterface() {propertyValues {
        AlphaCurve {keys []}
        BlueCurve {keys [<at 0.0028 1>,<at 0.0267 1 RCIM_Cubic>,<at 0.1061 1>,<at 0.1924 0.901>,<at 0.2694 0.574>,<at 0.3542 0.5781>,<at 0.6976 0.5448>,<at 0.9101 0.2448>]}
        GreenCurve {keys [
          <at 0.0028 0.6428>
          {arriveTangent=0.8673, interpMode=RCIM_Cubic, leaveTangent=0.8673, time=0.0267, value=0.8408}
          <at 0.1061 0.7324>
          <at 0.1924 0.8118>
          <at 0.2694 0.5537>
          <at 0.3542 0.5048>
          <at 0.6976 0.4752>
          <at 0.9101 0.2041>
         ]}
        RedCurve {keys [
          <at 0.0028 0.1861>
          {arriveTangent=2.9892, interpMode=RCIM_Cubic, leaveTangent=2.9892, time=0.0267, value=0.4676}
          <at 0.1061 0.4948>
          <at 0.1924 0.6664>
          <at 0.2694 0.4987>
          <at 0.3542 0.4405>
          <at 0.6976 0.4171>
          <at 0.9101 0.2221>
         ]}
       }}))
    ScaleA=true
    ScaleRGB=true
   )
  ]
  rendererClasses [&NiagaraSpriteRendererProperties]
  renderers [{rendererClass=&NiagaraSpriteRendererProperties}]
 }
}
summary {
 systemName=NS_Emote_StuffRasp_Flares
 emitters [
  {emitterName=Smoke, rendererClasses [&NiagaraSpriteRendererProperties]}
  {
   emitterName=Flames
   rendererClasses [
    &NiagaraSpriteRendererProperties
    &NiagaraLightRendererProperties
    &NiagaraMeshRendererProperties
   ]
  }
  {emitterName=Smoke_Ribbon, rendererClasses [&NiagaraRibbonRendererProperties]}
  {emitterName=Sparks, rendererClasses [&NiagaraSpriteRendererProperties]}
  {emitterName=Sparks2, rendererClasses [&NiagaraSpriteRendererProperties]}
 ]
 userVariables [
  User.bIsFrontEnd {type=NiagaraBool}
  User.bIsFrontEndPreview {type=NiagaraBool}
 ]
}
